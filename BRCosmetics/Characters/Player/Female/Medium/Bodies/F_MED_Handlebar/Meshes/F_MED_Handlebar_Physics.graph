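# Physics asset for the F_MED_Handlebar female medium player body cosmetic, covering its left and right dangling belt bones.
# Each side is a chain of box bodies from a master bone through four segments, with the master constrained to the thigh on that side.
# Joints lock twist and allow narrow swing that loosens toward the belt tips.
# Large thigh capsules and a tiny, unconstrained pelvis capsule are the only other bodies.
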PhysicsAsset /BRCosmetics/Characters/Player/Female/Medium/Bodies/F_MED_Handlebar/Meshes/F_MED_Handlebar_Physics
bodies {
 dyn_L_belt_0 [{shapeType=Box, extentX=7, extentY=3, extentZ=6.5, shapeName=dyn_L_belt_0_box, center=<2.7377 0 0>, rotation=<rot 90 0 0>}]
 dyn_L_belt_1 [{shapeType=Box, extentX=7, extentY=3, extentZ=6.5, shapeName=dyn_L_belt_1_box, center=<2.7652 0 0>, rotation=<rot 90 0 0>}]
 dyn_L_belt_2 [{shapeType=Box, extentX=7, extentY=3, extentZ=6.5, shapeName=dyn_L_belt_1_box, center=<2.7544 0 0>, rotation=<rot 90 0 0>}]
 dyn_L_belt_3 [{shapeType=Box, extentX=7, extentY=3, extentZ=11, shapeName=dyn_L_belt_1_box, center=<5.2453 0 0>, rotation=<rot 90 0 0>}]
 dyn_L_belt_master [{shapeType=Box, extentX=7, extentY=2.02, extentZ=6.5, shapeName=dyn_L_belt_0_box, center=<2.7377 0 0>, rotation=<rot 90 0 0>}]
 dyn_R_belt_0 [{shapeType=Box, extentX=7, extentY=3, extentZ=6.5, shapeName=dyn_R_belt_0_box, center=<-2.7377 0 0>, rotation=<rot 90 0 0>}]
 dyn_R_belt_1 [{shapeType=Box, extentX=7, extentY=3, extentZ=6.5, shapeName=dyn_R_belt_1_box, center=<-2.7652 0 0>, rotation=<rot 90 0 0>}]
 dyn_R_belt_2 [{shapeType=Box, extentX=7, extentY=3, extentZ=6.5, shapeName=dyn_L_belt_1_box, center=<-2.7544 0 0>, rotation=<rot 90 0 0>}]
 dyn_R_belt_3 [{shapeType=Box, extentX=7, extentY=3, extentZ=11, shapeName=dyn_L_belt_1_box, center=<-5.2453 0 0>, rotation=<rot 90 0 0>}]
 dyn_R_belt_master [{shapeType=Box, extentX=7, extentY=2.02, extentZ=6.5, shapeName=dyn_R_belt_0_box, center=<-2.7377 0 0>, rotation=<rot 90 0 0>}]
 pelvis [{shapeType=Capsule, length=1, radius=1, shapeName=pelvis_capsule}]
 thigh_l [{
   shapeType=Capsule
   length=45
   radius=15
   shapeName=thigh_l_capsule
   center=<-14.0799 -7.5225 -0.7372>
   rotation=<rot 89.6455 86.9391 90.6199>
  }]
 thigh_r [{shapeType=Capsule, length=45, radius=15, shapeName=thigh_r_capsule, center=<14.0799 7.5225 0.7372>, rotation=<rot 89.6455 86.939 90.6199>}]
}
constraints [
 {bone1Name=dyn_R_belt_2, bone2Name=dyn_R_belt_1, swing1LimitDegrees=15, swing2LimitDegrees=20, twistLimitDegrees=45, twistMotion=Locked}
 {bone1Name=dyn_R_belt_3, bone2Name=dyn_R_belt_2, swing1LimitDegrees=15, swing2LimitDegrees=20, twistLimitDegrees=45, twistMotion=Locked}
 {bone1Name=dyn_L_belt_2, bone2Name=dyn_L_belt_1, swing1LimitDegrees=15, swing2LimitDegrees=20, twistLimitDegrees=45, twistMotion=Locked}
 {bone1Name=dyn_L_belt_3, bone2Name=dyn_L_belt_2, swing1LimitDegrees=15, swing2LimitDegrees=20, twistLimitDegrees=45, twistMotion=Locked}
 {bone1Name=dyn_L_belt_1, bone2Name=dyn_L_belt_0, swing1LimitDegrees=15, swing2LimitDegrees=10, twistLimitDegrees=45, twistMotion=Locked}
 {bone1Name=dyn_R_belt_1, bone2Name=dyn_R_belt_0, swing1LimitDegrees=15, swing2LimitDegrees=10, twistLimitDegrees=45, twistMotion=Locked}
 {bone1Name=dyn_L_belt_0, bone2Name=dyn_L_belt_master, swing1LimitDegrees=5, swing2LimitDegrees=10, twistLimitDegrees=45, twistMotion=Locked}
 {bone1Name=dyn_R_belt_0, bone2Name=dyn_R_belt_master, swing1LimitDegrees=5, swing2LimitDegrees=10, twistLimitDegrees=45, twistMotion=Locked}
 {bone1Name=dyn_R_belt_master, bone2Name=thigh_r, swing1LimitDegrees=5, swing2LimitDegrees=10, twistLimitDegrees=45, twistMotion=Locked}
 {bone1Name=dyn_L_belt_master, bone2Name=thigh_l, swing1LimitDegrees=5, swing2LimitDegrees=10, twistLimitDegrees=45, twistMotion=Locked}
]
