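# Physics asset for the crouch variant of the F_MED_Handlebar female body cosmetic, defining physics bodies for its dangling left and right belt chains.
# Each three-segment capsule chain hangs from a master body locked in rotation to the pelvis, with limited swing and twist between segments and twist locked at the root.
# Large thigh capsules, rotationally free relative to the pelvis, provide leg volumes alongside the belts, likely as colliders.
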PhysicsAsset /BRCosmetics/Characters/Player/Female/Medium/Bodies/F_MED_Handlebar/Meshes/F_MED_Handlebar_Physics_Crouch
bodies {
 dyn_L_belt_1 [{shapeType=Capsule, length=5, radius=1.8, shapeName=dyn_L_belt_1_capsule, center=<1.9565 1.5777 0>, rotation=<rot 90 0 0>}]
 dyn_L_belt_1_master [{shapeType=Capsule, length=0.505, radius=0.505, shapeName=dyn_L_belt_1_master_capsule, rotation=<rot 0 -90 0>}]
 dyn_L_belt_2 [{shapeType=Capsule, length=5, radius=2, shapeName=dyn_L_belt_2_capsule, center=<1.8505 0.9814 0.0004>, rotation=<rot 90 -90.0001 -85.6594>}]
 dyn_L_belt_3 [{shapeType=Capsule, length=5, radius=2, shapeName=dyn_L_belt_3_capsule, center=<1.8586 0.4932 -0.0003>, rotation=<rot 90 0 5.7707>}]
 dyn_R_belt_1 [{shapeType=Capsule, length=5, radius=1.8, shapeName=dyn_R_belt_1_capsule, center=<-1.9565 -1.5777 0>, rotation=<rot 90 0 0>}]
 dyn_R_belt_1_master [{shapeType=Capsule, length=0.505, radius=0.505, shapeName=dyn_R_belt_1_master_capsule, rotation=<rot 0 -90 0>}]
 dyn_R_belt_2 [{shapeType=Capsule, length=5, radius=2, shapeName=dyn_R_belt_2_capsule, center=<-1.8505 -0.9814 -0.0004>, rotation=<rot 90 -90 -85.6594>}]
 dyn_R_belt_3 [{shapeType=Capsule, length=5, radius=2, shapeName=dyn_R_belt_3_capsule, center=<-1.8586 -0.4932 0.0003>, rotation=<rot 90 179.6419 -174.5874>}]
 pelvis [{shapeType=Capsule, length=1, radius=1, shapeName=pelvis_capsule}]
 thigh_l [{
   shapeType=Capsule
   length=30
   radius=15
   shapeName=thigh_l_capsule
   center=<-14.0799 -7.5225 -0.7372>
   rotation=<rot 89.6455 86.9391 90.6199>
  }]
 thigh_r [{shapeType=Capsule, length=30, radius=15, shapeName=thigh_r_capsule, center=<14.0799 7.5225 0.7372>, rotation=<rot 89.6455 86.939 90.6199>}]
}
constraints [
 {bone1Name=dyn_R_belt_2, bone2Name=dyn_R_belt_1, swing1LimitDegrees=20, swing2LimitDegrees=20, twistLimitDegrees=20}
 {bone1Name=dyn_R_belt_3, bone2Name=dyn_R_belt_2, swing1LimitDegrees=45, swing2LimitDegrees=45, twistLimitDegrees=20}
 {bone1Name=dyn_L_belt_2, bone2Name=dyn_L_belt_1, swing1LimitDegrees=20, swing2LimitDegrees=20, twistLimitDegrees=20}
 {bone1Name=dyn_L_belt_3, bone2Name=dyn_L_belt_2, swing1LimitDegrees=45, swing2LimitDegrees=45, twistLimitDegrees=20}
 {
  bone1Name=dyn_L_belt_1
  bone2Name=dyn_L_belt_1_master
  swing1LimitDegrees=45
  swing2LimitDegrees=45
  twistLimitDegrees=45
  twistMotion=Locked
 }
 {
  bone1Name=dyn_R_belt_1
  bone2Name=dyn_R_belt_1_master
  swing1LimitDegrees=45
  swing2LimitDegrees=45
  twistLimitDegrees=45
  twistMotion=Locked
 }
 {
  bone1Name=dyn_L_belt_1_master
  bone2Name=pelvis
  swing1LimitDegrees=45
  swing1Motion=Locked
  swing2LimitDegrees=45
  swing2Motion=Locked
  twistLimitDegrees=45
  twistMotion=Locked
 }
 {
  bone1Name=dyn_R_belt_1_master
  bone2Name=pelvis
  swing1LimitDegrees=45
  swing1Motion=Locked
  swing2LimitDegrees=45
  swing2Motion=Locked
  twistLimitDegrees=45
  twistMotion=Locked
 }
 {
  bone1Name=thigh_l
  bone2Name=pelvis
  swing1LimitDegrees=45
  swing1Motion=Free
  swing2LimitDegrees=45
  swing2Motion=Free
  twistLimitDegrees=45
  twistMotion=Free
 }
 {
  bone1Name=thigh_r
  bone2Name=pelvis
  swing1LimitDegrees=45
  swing1Motion=Free
  swing2LimitDegrees=45
  swing2Motion=Free
  twistLimitDegrees=45
  twistMotion=Free
 }
]
