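# Post-process style AnimBP for the F_MED_Handlebar player body, adding secondary motion to its scarf, hood, pouch and belts.
# Copied parent pose runs through Control Rig, weighted constraints and AnimDynamics into a cached pose.
# Crouching selects between two belt RigidBody branches with separate physics assets; the standing one adds front-end thigh constraints.
# The event graph drives scarf alphas from neck rotation, standing-branch gravity, and material WPO with skydive wind.
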AnimBlueprint /BRCosmetics/Characters/Player/Female/Medium/Bodies/F_MED_Handlebar/Meshes/F_MED_Handlebar_AnimBP
target_skeleton=&/F_MED_Handlebar_Skeleton
graphs [
 AnimGraph {nodes [
   n1 = "Constraint - dyn_L_belt_0 (P)"(ComponentPose=Local To Component(LocalPose=Use cached pose 'startPose'()), Alpha=1, ConstraintWeights_0=1) {
    alphaBoolBlend {blendOption=Linear}
    boneToModify {boneName=dyn_L_belt_0}
    constraintSetup [{targetBone {boneName=thigh_l}}]
    constraintWeights [1]
   }
   n2 = "Constraint - dyn_R_belt_0 (P)"(ComponentPose=@n1, Alpha=1, ConstraintWeights_0=1) {
    alphaBoolBlend {blendOption=Linear}
    boneToModify {boneName=dyn_R_belt_0}
    constraintSetup [{targetBone {boneName=thigh_r}}]
    constraintWeights [1]
   }
   n3 = "Transform (Modify) Bone - Bone: dyn_R_belt_const"(ComponentPose=@n2, Alpha=1, Translation="0,0,0", Rotation="-0.000,10,-0.000", Scale="1,1,1") {
    rotationMode=BMM_Additive
    rotationSpace=BCS_BoneSpace
    alphaBoolBlend {blendOption=Linear}
    boneToModify {boneName=dyn_R_belt_const}
    rotation {yaw=10}
   }
   n4 = RigidBody(ComponentPose=@n3) {
    bClampLinearTranslationLimitToRefPose=true
    bOverrideWorldGravity=true
    overridePhysicsAsset=&/F_MED_Handlebar_Physics_Crouch
    simulationSpace=BaseBoneSpace
    alphaBoolBlend {blendOption=Linear}
    baseBoneRef {boneName=pelvis}
    componentLinearAccScale=<0.2 0.2 0.2>
    componentLinearVelScale=<0.3 0.3 0.3>
    overrideWorldGravity {z=-980}
    simSpaceSettings {maxAngularAcceleration=100, maxAngularVelocity=100, maxLinearAcceleration=100, maxLinearVelocity=100, worldAlpha=1}
   }
   n5 = "Transform (Modify) Bone - Bone: dyn_L_belt_const"(
    ComponentPose=Local To Component(LocalPose=Use cached pose 'startPose'())
    Alpha=1
    Translation="0,0,0"
    Rotation="-0.000,10,-0.000"
    Scale="1,1,1"
   ) {
    rotationMode=BMM_Additive
    rotationSpace=BCS_BoneSpace
    alphaBoolBlend {blendOption=Linear}
    boneToModify {boneName=dyn_L_belt_const}
    rotation {yaw=10}
   }
   n6 = "Transform (Modify) Bone - Bone: dyn_R_belt_const"(ComponentPose=@n5, Alpha=1, Translation="0,0,0", Rotation="-0.000,10,-0.000", Scale="1,1,1") {
    rotationMode=BMM_Additive
    rotationSpace=BCS_BoneSpace
    alphaBoolBlend {blendOption=Linear}
    boneToModify {boneName=dyn_R_belt_const}
    rotation {yaw=10}
   }
   n7 = RigidBody!(
    ComponentPose=@n6
    bAlphaBoolEnabled:bool
    OverrideWorldGravity:Vector=Get General Gravity Override() -> (General Gravity Override:Vector)
    ExternalForce
   ) -> (Pose) {
    alphaInputType=Bool
    bClampLinearTranslationLimitToRefPose=true
    bOverrideWorldGravity=true
    overridePhysicsAsset=&/F_MED_Handlebar_Physics
    simulationSpace=BaseBoneSpace
    alphaBoolBlend {blendOption=Linear}
    baseBoneRef {boneName=root}
    componentLinearAccScale=<0.2 0.2 0.2>
    componentLinearVelScale=<0.3 0.3 0.3>
    overrideWorldGravity {z=-980}
   }
   n8 = Get FrontEnd_Idle_On() -> (FrontEnd_Idle_On:bool)
   n9 = "Constraint - dyn_L_belt_1_master (P)"!(ComponentPose=@n7, bAlphaBoolEnabled:bool=@n8, ConstraintWeights_0=1) -> (Pose) {
    alphaInputType=Bool
    alphaBoolBlend {blendOption=Linear}
    boneToModify {boneName=dyn_L_belt_1_master}
    constraintSetup [{targetBone {boneName=thigh_l}}]
    constraintWeights [1]
   }
   n10 = "Constraint - dyn_R_belt_1_master (P)"!(ComponentPose=@n9, bAlphaBoolEnabled:bool=@n8, ConstraintWeights_0=1) -> (Pose) {
    alphaInputType=Bool
    alphaBoolBlend {blendOption=Linear}
    boneToModify {boneName=dyn_R_belt_1_master}
    constraintSetup [{targetBone {boneName=thigh_r}}]
    constraintWeights [1]
   }
   n11 = AND Boolean(A=Get bIsInFrontEnd(), B=Get isLOD1+() -> (isLOD1+:bool))
   n12 = "Constraint - dyn_L_belt_2 (P)"!(ComponentPose=@n10, bAlphaBoolEnabled:bool=@n11, ConstraintWeights_0=1) -> (Pose) {
    alphaInputType=Bool
    alphaBoolBlend {blendOption=Linear}
    boneToModify {boneName=dyn_L_belt_2}
    constraintSetup [{targetBone {boneName=thigh_l}}]
    constraintWeights [1]
   }
   n13 = "Constraint - dyn_R_belt_2 (P)"!(ComponentPose=@n12, bAlphaBoolEnabled:bool=@n11, ConstraintWeights_0=1) -> (Pose) {
    alphaInputType=Bool
    alphaBoolBlend {blendOption=Linear}
    boneToModify {boneName=dyn_R_belt_2}
    constraintSetup [{targetBone {boneName=thigh_r}}]
    constraintWeights [1]
   }
   n14 = "Constraint - dyn_L_belt_3 (P)"!(ComponentPose=@n13, bAlphaBoolEnabled:bool=@n11, ConstraintWeights_0=1) -> (Pose) {
    alphaInputType=Bool
    alphaBoolBlend {blendOption=Linear}
    boneToModify {boneName=dyn_L_belt_3}
    constraintSetup [{targetBone {boneName=thigh_l}}]
    constraintWeights [1]
   }
   n15 = "Constraint - dyn_R_belt_3 (P)"!(ComponentPose=@n14, bAlphaBoolEnabled:bool=@n11, ConstraintWeights_0=1) -> (Pose) {
    alphaInputType=Bool
    alphaBoolBlend {blendOption=Linear}
    boneToModify {boneName=dyn_R_belt_3}
    constraintSetup [{targetBone {boneName=thigh_r}}]
    constraintWeights [1]
   }
   Output Pose(Result=Blend Poses by bool(BlendPose_0=Component To Local(ComponentPose=@n4), BlendPose_1=Component To Local(ComponentPose=@n15), bActiveValue=Get bIsCrouching()) {blendType=Linear}) {layerGroup=None}
   n16 = Control Rig(Source=Copy Pose From Mesh() {bUseAttachedParent=true}) {
    alphaBoolBlend {blendOption=Linear}
    controlRigAssetReference {blueprintRigClass=&/F_MED_Handlebar_CtrlRig.F_MED_Handlebar_CtrlRig_C}
   }
   n17 = "Constraint - dyn_pouch (PPPPP)"!(
    ComponentPose=Local To Component(LocalPose=@n16)
    Alpha=1
    ConstraintWeights_0=0.1707
    ConstraintWeights_1:float=0.7795
    ConstraintWeights_2:float=0.015
    ConstraintWeights_3:float=0.0185
    ConstraintWeights_4:float=0.0163
   ) -> (Pose) {
    alphaBoolBlend {blendOption=Linear}
    boneToModify {boneName=dyn_pouch}
    constraintSetup [
     {targetBone {boneName=pelvis}}
     {targetBone {boneName=spine_01}}
     {targetBone {boneName=thigh_l}}
     {targetBone {boneName=deform_groin_l}}
     {targetBone {boneName=deform_glute_l}}
    ]
    constraintWeights [1,1,1,1,1]
   }
   n18 = "Constraint - dyn_rotateScarf (P)"(ComponentPose=@n17, Alpha=Get hoodAlphaBack() -> (hoodAlphaBack:double), ConstraintWeights_0=1) {
    alphaBoolBlend {blendOption=Linear}
    boneToModify {boneName=dyn_rotateScarf}
    constraintSetup [{targetBone {boneName=neck_01}}]
    constraintWeights [1]
   }
   n19 = "Transform (Modify) Bone - Bone: dyn_rotateScarf"!(
    ComponentPose=@n18
    bAlphaBoolEnabled:bool=Get skydiveFloatOrDive() -> (skydiveFloatOrDive:bool)
    Translation="0,0,0"
    Rotation="0,0.000,-30"
    Scale="1,1,1"
   ) -> (Pose) {
    alphaInputType=Bool
    rotationMode=BMM_Additive
    alphaBoolBlend {blendOption=Linear}
    boneToModify {boneName=dyn_rotateScarf}
    rotation {roll=-30}
   }
   n20 = "Constraint - dyn_rotateScarf_L (P)"(ComponentPose=@n19, Alpha=Get Hood L Alpha() -> (Hood L Alpha:double), ConstraintWeights_0=1) {
    alphaBoolBlend {blendOption=Linear}
    boneToModify {boneName=dyn_rotateScarf_L}
    constraintSetup [{targetBone {boneName=neck_01}}]
    constraintWeights [1]
   }
   n21 = Get scarf R correction() -> (scarf R correction:double)
   n22 = "Transform (Modify) Bone - Bone: dyn_rotateScarf_R"(ComponentPose=@n20, Alpha=@n21, Translation="0,0,1.1", Rotation="0,0,0", Scale="1,1,1.2") {
    rotationMode=BMM_Additive
    rotationSpace=BCS_BoneSpace
    scaleMode=BMM_Additive
    scaleSpace=BCS_BoneSpace
    translationMode=BMM_Additive
    translationSpace=BCS_BoneSpace
    alphaBoolBlend {blendOption=Linear}
    boneToModify {boneName=dyn_rotateScarf_R}
    scale {z=1.2}
    translation {z=1.1}
   }
   n23 = "Transform (Modify) Bone - Bone: dyn_rotateScarfForward"(ComponentPose=@n22, Alpha=@n21, Translation="0,2.3,0", Rotation="0,0,0", Scale="1,1,1") {
    rotationMode=BMM_Additive
    rotationSpace=BCS_BoneSpace
    scaleMode=BMM_Additive
    scaleSpace=BCS_BoneSpace
    translationMode=BMM_Additive
    translationSpace=BCS_BoneSpace
    alphaBoolBlend {blendOption=Linear}
    boneToModify {boneName=dyn_rotateScarfForward}
    translation {y=2.3}
   }
   n24 = "Constraint - dyn_rotateScarfForward (P)"(ComponentPose=@n23, Alpha=Get HoodAlphaFront() -> (HoodAlphaFront:double), ConstraintWeights_0=1) {
    alphaBoolBlend {blendOption=Linear}
    boneToModify {boneName=dyn_rotateScarfForward}
    constraintSetup [{targetBone {boneName=neck_01}}]
    constraintWeights [1]
   }
   n25 = Anim Dynamics - Bone: dyn_pouch!(ComponentPose=@n24, bAlphaBoolEnabled:bool=Get Is NOT Front End() -> (Is NOT Front End:bool), ExternalForce="0,0,0") -> (Pose) {
    alphaInputType=Bool
    angularDampingOverride=1
    angularSpringConstant=2000
    bAngularSpring=true
    bOverrideAngularDamping=true
    bUseGravityOverride=true
    gravityScale=0.35
    simulationSpace=BoneRelative
    alphaBoolBlend {blendOption=Linear}
    boundBone {boneName=dyn_pouch}
    componentLinearAccScale=<0.9 0.9 0.9>
    componentLinearVelScale=<0.9 0.9 0.9>
    gravityOverride {z=-980}
    physicsBodyDefinitions [{
      boundBone {boneName=dyn_pouch}
      boxExtents=<25 25 25>
      constraintSetup {angularLimitsMax=<0 20 5>, angularLimitsMin=<0 0 -5>, angularTarget=<1 0 0>}
      localJointOffset=<2 0 0>
     }]
    relativeSpaceBone {boneName=pelvis}
   }
   n26 = Anim Dynamics - Bone: dyn_rotateScarf(ComponentPose=@n25, Alpha=1, ExternalForce="0,0,0") {
    angularDampingOverride=1
    angularSpringConstant=200
    bAngularSpring=true
    bLinearSpring=true
    bOverrideLinearDamping=true
    bUseGravityOverride=true
    gravityScale=0.35
    linearDampingOverride=1
    linearSpringConstant=200
    simulationSpace=RootRelative
    alphaBoolBlend {blendOption=Linear}
    boundBone {boneName=dyn_rotateScarf}
    componentLinearAccScale=<1 1 1>
    componentLinearVelScale=<1 1 1>
    gravityOverride {z=-2300}
    physicsBodyDefinitions [{
      boundBone {boneName=dyn_rotateScarf}
      boxExtents=<200 200 200>
      constraintSetup {angularLimitsMax=<0 0 20>, angularLimitsMin=<0 0 -20>, angularTarget=<0 0 1>, linearAxesMax=<3 0 1>, linearAxesMin=<-0.5 0 0>}
      localJointOffset=<0 0 0>
     }]
   }
   n27 = Anim Dynamics - Bone: dyn_hood(ComponentPose=@n26, Alpha=1, ExternalForce="0,0,0") {
    angularDampingOverride=1
    angularSpringConstant=20000
    bLinearSpring=true
    bOverrideLinearDamping=true
    bUseGravityOverride=true
    gravityScale=0.35
    linearDampingOverride=1
    linearSpringConstant=200
    simulationSpace=RootRelative
    alphaBoolBlend {blendOption=Linear}
    boundBone {boneName=dyn_hood}
    componentLinearAccScale=<1 1 1>
    componentLinearVelScale=<1 1 1>
    gravityOverride {z=-2000}
    physicsBodyDefinitions [{
      boundBone {boneName=dyn_hood}
      boxExtents=<100 100 100>
      constraintSetup {angularLimitsMax=<0 0 0>, angularLimitsMin=<0 0 0>, angularTarget=<0 0 0>, linearAxesMax=<1 0.5 1>, linearAxesMin=<-1 -2 0>}
      localJointOffset=<0 0 0>
     }]
   }
   n28 = Anim Dynamics - Bone: dyn_L_WPO(ComponentPose=@n27, Alpha=1, ExternalForce="0,0,0") {
    angularDampingOverride=1
    angularSpringConstant=20000
    bLinearSpring=true
    bOverrideLinearDamping=true
    bUseGravityOverride=true
    gravityScale=0.35
    linearDampingOverride=1
    linearSpringConstant=100
    simulationSpace=BoneRelative
    alphaBoolBlend {blendOption=Linear}
    alphaScaleBiasClamp {scale=0.7}
    boundBone {boneName=dyn_L_WPO}
    componentLinearAccScale=<0.8 0.8 0.8>
    componentLinearVelScale=<0.8 0.8 0.8>
    gravityOverride {z=-2000}
    physicsBodyDefinitions [{
      boundBone {boneName=dyn_L_WPO}
      boxExtents=<190 190 190>
      constraintSetup {angularLimitsMax=<0 0 0>, angularLimitsMin=<0 0 0>, angularTarget=<0 0 0>, linearAxesMax=<1 1 4>, linearAxesMin=<-1 -1 0>}
      localJointOffset=<0 0 0>
     }]
    relativeSpaceBone {boneName=calf_l}
   }
   n29 = Anim Dynamics - Bone: dyn_R_WPO(ComponentPose=@n28, Alpha=1, ExternalForce="0,0,0") {
    angularDampingOverride=1
    angularSpringConstant=20000
    bLinearSpring=true
    bOverrideLinearDamping=true
    bUseGravityOverride=true
    gravityScale=0.35
    linearDampingOverride=1
    linearSpringConstant=100
    simulationSpace=BoneRelative
    alphaBoolBlend {blendOption=Linear}
    alphaScaleBiasClamp {scale=0.7}
    boundBone {boneName=dyn_R_WPO}
    componentLinearAccScale=<0.8 0.8 0.8>
    componentLinearVelScale=<0.8 0.8 0.8>
    gravityOverride {z=-2000}
    physicsBodyDefinitions [{
      boundBone {boneName=dyn_R_WPO}
      boxExtents=<190 190 190>
      constraintSetup {angularLimitsMax=<0 0 0>, angularLimitsMin=<0 0 0>, angularTarget=<0 0 0>, linearAxesMax=<1 1 0>, linearAxesMin=<-1 -1 -4.5>}
      localJointOffset=<0 0 0>
     }]
    relativeSpaceBone {boneName=calf_r}
   }
   n30 = "Constraint - attach_backpack (PPPPPP)"!(
    ComponentPose=@n29
    Alpha=1
    ConstraintWeights_0=0.0878
    ConstraintWeights_1:float=0.3242
    ConstraintWeights_2:float=0.1104
    ConstraintWeights_3:float=0.147
    ConstraintWeights_4:float=0.1731
    ConstraintWeights_5:float=0.1576
   ) -> (Pose) {
    alphaBoolBlend {blendOption=Linear}
    boneToModify {boneName=attach_backpack}
    constraintSetup [
     {targetBone {boneName=spine_03}}
     {targetBone {boneName=spine_04}}
     {targetBone {boneName=clavicle_l}}
     {targetBone {boneName=spine_02}}
     {targetBone {boneName=clavicle_r}}
     {targetBone {boneName=spine_05}}
    ]
    constraintWeights [1,1,1,1,1,1]
   }
   n31 = "Constraint - attach_cape (PPPPPP)"!(
    ComponentPose=@n30
    Alpha=1
    ConstraintWeights_0=0.0163
    ConstraintWeights_1:float=0.0896
    ConstraintWeights_2:float=0.1076
    ConstraintWeights_3:float=0.0245
    ConstraintWeights_4:float=0.1314
    ConstraintWeights_5:float=0.6306
   ) -> (Pose) {
    alphaBoolBlend {blendOption=Linear}
    boneToModify {boneName=attach_cape}
    constraintSetup [
     {targetBone {boneName=spine_03}}
     {targetBone {boneName=spine_04}}
     {targetBone {boneName=clavicle_l}}
     {targetBone {boneName=spine_02}}
     {targetBone {boneName=clavicle_r}}
     {targetBone {boneName=spine_05}}
    ]
    constraintWeights [1,1,1,1,1,1]
   }
   n32 = "Constraint - dyn_L_buckle_const (PPPPPPPP)"!(
    ComponentPose=@n31
    Alpha=0.5
    ConstraintWeights_0=0.4353
    ConstraintWeights_1:float=0.0014
    ConstraintWeights_2:float=0.0077
    ConstraintWeights_3:float=0.4134
    ConstraintWeights_4:float=0.0289
    ConstraintWeights_5:float=0.0007
    ConstraintWeights_6:float=0.111
    ConstraintWeights_7:float=0.0016
   ) -> (Pose) {
    alphaBoolBlend {blendOption=Linear}
    boneToModify {boneName=dyn_L_buckle_const}
    constraintSetup [
     {targetBone {boneName=pelvis}}
     {targetBone {boneName=thigh_twist_01_l}}
     {targetBone {boneName=deform_glute_l}}
     {targetBone {boneName=spine_01}}
     {targetBone {boneName=thigh_l}}
     {targetBone {boneName=thigh_r}}
     {targetBone {boneName=deform_groin_l}}
     {targetBone {boneName=deform_groin_r}}
    ]
    constraintWeights [1,1,1,1,1,1,1,1]
   }
   n33 = "Constraint - dyn_R_buckle_const (PPPPPPPP)"!(
    ComponentPose=@n32
    Alpha=0.5
    ConstraintWeights_0=0.3963
    ConstraintWeights_1:float=0.0103
    ConstraintWeights_2:float=0.4716
    ConstraintWeights_3:float=0
    ConstraintWeights_4:float=0.0345
    ConstraintWeights_5:float=0
    ConstraintWeights_6:float=0
    ConstraintWeights_7:float=0.0872
   ) -> (Pose) {
    alphaBoolBlend {blendOption=Linear}
    boneToModify {boneName=dyn_R_buckle_const}
    constraintSetup [
     {targetBone {boneName=pelvis}}
     {targetBone {boneName=deform_glute_r}}
     {targetBone {boneName=spine_01}}
     {targetBone {boneName=thigh_l}}
     {targetBone {boneName=thigh_r}}
     {targetBone {boneName=thigh_twist_01_r}}
     {targetBone {boneName=deform_groin_l}}
     {targetBone {boneName=deform_groin_r}}
    ]
    constraintWeights [1,1,1,1,1,1,1,1]
   }
   Save cached pose 'startPose'(Pose=Component To Local(ComponentPose=@n33)) {cacheName=startPose}
  ]}
 EventGraph {nodes [
   Delay(Duration=0.2)
   > ResetDynamics(InTeleportType=TeleportPhysics)
   > ResetDynamics(InTeleportType=ResetPhysics)
   n1 = Event BlueprintUpdateAnimation()
   > n2 = Sequence!(execute) -> (then_0, then_1, then_2:Exec, then_3:Exec, then_4:Exec, then_5:Exec)
   > n3 = Branch(Condition=Get skydiveFloatOrDive())
   > Set hoodAlphaBack(hoodAlphaBack:double=0) -> (Output_Get:double)
   n4 = GetOwnerSkeletalMeshComponent()
   n5 = BreakRotator(InRot=BreakTransform(InTransform=GetDeltaTransformFromRefPose(self=@n4, BoneName=neck_01, BaseName=spine_05)).Rotation)
   Set hoodAlphaBack(execute=@n3.else, hoodAlphaBack=MapRangeClamped(Value=@n5.Yaw, InRangeA=-10, InRangeB=-30))
   > Set HoodAlphaFront(HoodAlphaFront:double=MapRangeClamped(Value=@n5.Yaw, InRangeA=-1, InRangeB=20)) -> (Output_Get:double)
   > Set Hood L Alpha(Hood L Alpha:double=MapRangeClamped(Value=@n5.Pitch, InRangeA=0, InRangeB=-10)) -> (Output_Get:double)
   > Set Hood R Alpha(Hood R Alpha:double=MapRangeClamped(Value=@n5.Pitch, InRangeA=0, InRangeB=30)) -> (Output_Get:double)
   > Set scarf R correction(scarf R correction:double=float + float(A=Get Hood R Alpha() -> (Hood R Alpha:double), B=Get HoodAlphaFront())) -> (Output_Get:double)
   > Set Is NOT Front End(Is NOT Front End:bool=NOT Boolean(A=Get bIsInFrontEnd())) -> (Output_Get:bool)
   > Set isLOD1+(isLOD1+:bool="integer >= integer"(A=Get CurrentLOD(), B=1)) -> (Output_Get:bool)
   Set skydiveFloatOrDive(execute=@n2.then_1, skydiveFloatOrDive:bool=OR Boolean(A=Get bIsSkydiveFloating(), B=Get bIsSkydiveDiveMode())) -> (Output_Get:bool)
   Is Valid(exec=@n2.then_2, InputObject=Get playerPawn() -> (playerPawn:Fort Player Pawn Athena Object Reference))
   > Set WPO Globals Function(
    Material=Get Material Instance Dynamic() -> (Material Instance Dynamic:Material Instance Dynamic Object Reference)
    Skeletal Mesh=Get SkelMesh()
    "Local Noise Transform  Location"=Get Local Noise Transform Location() -> (Local Noise Transform Location:Vector)
    Local Noise Offset=Get Local Noise Offset()
   )
   > Set WPO Regional Function!(
    execute
    Material:Material Instance Dynamic Object Reference=Get Material Instance Dynamic()
    Overlay Material:Material Instance Dynamic Object Reference
    Skeletal Mesh:Skeletal Mesh Component Object Reference=Get SkelMesh()
    Velocity Transform:Name=root
    Velocity Interp Speed:double=30
    Collision Transform:Name=None
    Collision Transform Axis:EAxis Enum
    Invert Collision Transform Axis:bool=false
    Delta Time X:double=@n1.DeltaTimeX
    Previous Velocity Transform Location:"Vector (by ref)"=Get R1 Transform Location() -> (R1 Transform Location:Vector)
    Previous Velocity Transform Velocity:"Vector (by ref)"=Get R1 Transform Velocity() -> (R1 Transform Velocity:Vector)
    Region:int=1
   ) -> (then)
   > n6 = Branch(Condition=Get bIsSkyDiving())
   > n7 = Set WPO Pho Wind Intensity(WPO Pho Wind Intensity=1)
   SetScalarParameterValue(execute=[@n8.then, @n7.then], self=Get Material Instance Dynamic(), ParameterName=ClothSim_WindStrength, Value=Get WPO Pho Wind Intensity())
   n8 = Set WPO Pho Wind Intensity(execute=@n6.else, WPO Pho Wind Intensity=0)
   n9 = Branch(execute=@n2.then_3, Condition=Get bIsCrouching())
   > Set crouching(crouching:bool=true) -> (Output_Get:bool)
   Set crouching(execute=@n9.else, crouching=false)
   n10 = Gravity Override State Function DEPRECATED(
    execute=@n2.then_4
    CharacterPartAnimInstance=Self-Reference()
    GameTimeInSeconds=GetGameTimeInSeconds()
    GravityOverrideParams=Get Gravity Override Params()
   )
   > Set General Gravity Override(General Gravity Override:Vector=@n10) -> (Output_Get:Vector)
   > Branch(Condition=Get bIsFalling())
   > Set General Gravity Override(General Gravity Override="0, 0,-500")
   Is Valid(exec=@n2.then_5, InputObject=Get Anim_Inst() -> (Anim_Inst:Anim Instance Object Reference))
   n11 = Get bIsPlayingEmote()
   > Set FrontEnd_Idle_On(FrontEnd_Idle_On:bool=AND Boolean(A=Get bIsInFrontEnd(), B=NOT Boolean(A=@n11))) -> (Output_Get:bool)
   Event BlueprintInitializeAnimation()
   > n12 = Sequence!(execute) -> (then_0, then_1, then_2)
   n13 = GetOwnerSkeletalMeshComponent()
   > Set SkelMesh(SkelMesh=@n13)
   n14 = TryGetPawnOwner()
   > Is Valid(InputObject=@n14)
   > n15 = Cast To FortPlayerPawnAthena(Object=@n14)
   > Set playerPawn(playerPawn:Fort Player Pawn Athena Object Reference=@n15) -> (Output_Get:Fort Player Pawn Athena Object Reference)
   n16 = Set Skel_Mesh(execute=@n12.then_1, Skel_Mesh=@n13)
   > Set Anim_Inst(Anim_Inst:Anim Instance Object Reference=GetAnimInstance(self=@n16)) -> (Output_Get:Anim Instance Object Reference)
   Event OnCharacterCustomizationLoaded()
   > n17 = Set SkelMesh(SkelMesh=GetPartSkeletalMeshComponent())
   > n18 = Get Dynamic Material Instance(SkeletalMeshComponent=@n17)
   > Set Material Instance Dynamic(Material Instance Dynamic:Material Instance Dynamic Object Reference=@n18) -> (Output_Get:Material Instance Dynamic Object Reference)
   Event OnEndCrouching()
  ]}
]
variables [
 Anim_Inst {type=AnimInstance, default=null}
 crouching {type=boolean}
 Current LOD {type=integer}
 FrontEnd_Idle_On {type=boolean}
 General Gravity Override {type=Vector}
 Gravity Override Params {
  type=GravityOverrideParamsStruct
  default {onGround_Falling {gravityOverride=<0 0 -500>}, onGround_Moving {gravityOverride=<0 0 -3000>}, tactical_Sprint {gravityOverride=<0 0 -3000>}}
 }
 Hood L Alpha {type=number}
 Hood R Alpha {type=number}
 hoodAlphaBack {type=number}
 HoodAlphaFront {type=number}
 Is NOT Front End {type=boolean}
 isLOD1+ {type=boolean}
 Local Noise Offset {type=Vector}
 Local Noise Transform Location {type=Vector}
 Material Instance Dynamic {type=MaterialInstanceDynamic}
 playerPawn {type=FortPlayerPawnAthena}
 R1 Transform Location {type=Vector}
 R1 Transform Velocity {type=Vector}
 scarf R correction {type=number}
 Skel_Mesh {type=SkeletalMeshComponent}
 SkelMesh {type=SkeletalMeshComponent}
 skydiveFloatOrDive {type=boolean}
 WPO Pho Wind Intensity {type=number}
]
