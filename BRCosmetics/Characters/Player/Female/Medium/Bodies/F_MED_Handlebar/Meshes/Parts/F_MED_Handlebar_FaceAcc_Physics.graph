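# Physics asset for the face accessory part of the F_MED_Handlebar female medium player body, rigging its hair for secondary motion.
# Capsule chains for left and right front hair and a four-link strand hang from the head with tight swing and twist limits.
# Single back and right-side hair bodies lock twist, and the side piece also locks one swing axis.
# Head, upper spine and clavicle capsules supply body shapes, joined by free constraints.
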PhysicsAsset /BRCosmetics/Characters/Player/Female/Medium/Bodies/F_MED_Handlebar/Meshes/Parts/F_MED_Handlebar_FaceAcc_Physics
bodies {
 clavicle_l [{shapeType=Capsule, length=10, radius=6, shapeName=clavicle_l_capsule, center=<3.0645 0 -1.8599>, rotation=<rot 90 0 0>}]
 clavicle_r [{shapeType=Capsule, length=10, radius=6, shapeName=clavicle_l_capsule, center=<-3.0645 0 1.8599>, rotation=<rot 90 0 0>}]
 dyn_L_frontHair_1 [{shapeType=Capsule, length=4, radius=2.5, shapeName=dyn_L_frontHair_1_capsule, center=<3.1628 0 0>, rotation=<rot 90 0 0>}]
 dyn_L_frontHair_2 [{shapeType=Capsule, length=4, radius=1.8, shapeName=dyn_L_frontHair_2_capsule, center=<3.1712 0 0>, rotation=<rot 90 0 0>}]
 dyn_L_frontHair_3 [{shapeType=Capsule, length=4, radius=2, shapeName=dyn_L_frontHair_3_capsule, center=<3.1024 -0.3156 0>, rotation=<rot 90 -176.4237 -164.1392>}]
 dyn_R_frontHair_1 [{shapeType=Capsule, length=4, radius=2.5, shapeName=dyn_R_frontHair_1_capsule, center=<3.1658 0 0>, rotation=<rot 90 0 0>}]
 dyn_R_frontHair_2 [{shapeType=Capsule, length=4, radius=1.8, shapeName=dyn_R_frontHair_2_capsule, center=<3.1626 0 0>, rotation=<rot 90 0 0>}]
 dyn_R_frontHair_3 [{
   shapeType=Capsule
   length=7
   radius=2
   shapeName=dyn_R_frontHair_3_capsule
   center=<2.1112 -1.2354 3.1833>
   rotation=<rot 72.7867 157.0577 -173.4671>
  }]
 dyn_R_side_hair [{shapeType=Capsule, length=5, radius=2.5, shapeName=dyn_R_hair_capsule, center=<2 0 0>, rotation=<rot 90 0 0>}]
 dyn_backHair [{shapeType=Capsule, length=9, radius=2.5, shapeName=dyn_backHair_capsule, center=<4 0 0>, rotation=<rot 90 0 0>}]
 dyn_strand_1 [{shapeType=Capsule, length=3, radius=1.3, shapeName=dyn_strand_1_capsule, center=<1.3 0 0>, rotation=<rot 90 0 0>}]
 dyn_strand_2 [{shapeType=Capsule, length=3, radius=1, shapeName=dyn_strand_2_capsule, center=<1.2013 0 0>, rotation=<rot 90 0 0>}]
 dyn_strand_3 [{shapeType=Capsule, length=3, radius=1, shapeName=dyn_strand_3_capsule, center=<1.2727 0 0>, rotation=<rot 90 0 0>}]
 dyn_strand_4 [{shapeType=Capsule, length=3, radius=1, shapeName=dyn_strand_4_capsule, center=<1.2806 0 0>, rotation=<rot 90 0 0>}]
 head [
  {shapeType=Capsule, length=4, radius=7, shapeName=head_capsule, center=<-0.0105 1.2729 0>, rotation=<rot 90 0 0>}
  {shapeType=Capsule, length=1.6511, radius=1.3, shapeName=head_capsule, center=<0.1441 9.1097 0>, rotation=<rot 90 180 -159.9998>}
 ]
 spine_05 [
  {shapeType=Capsule, length=8.8209, radius=6.9478, shapeName=spine_05_capsule, center=<10.6319 -6.6383 0.9692>}
  {shapeType=Capsule, length=2, radius=11, shapeName=spine_05_capsule, center=<10.5544 -2.1008 0>, rotation=<rot 0 0 90>}
 ]
}
constraints [
 {bone1Name=dyn_L_frontHair_1, bone2Name=head, swing1LimitDegrees=20, swing2LimitDegrees=20, twistLimitDegrees=10}
 {bone1Name=dyn_L_frontHair_2, bone2Name=dyn_L_frontHair_1, swing1LimitDegrees=20, swing2LimitDegrees=20, twistLimitDegrees=10}
 {bone1Name=dyn_L_frontHair_3, bone2Name=dyn_L_frontHair_2, swing1LimitDegrees=20, swing2LimitDegrees=20, twistLimitDegrees=10}
 {bone1Name=dyn_R_frontHair_1, bone2Name=head, swing1LimitDegrees=20, swing2LimitDegrees=20, twistLimitDegrees=10}
 {bone1Name=dyn_R_frontHair_2, bone2Name=dyn_R_frontHair_1, swing1LimitDegrees=20, swing2LimitDegrees=20, twistLimitDegrees=10}
 {bone1Name=dyn_R_frontHair_3, bone2Name=dyn_R_frontHair_2, swing1LimitDegrees=20, swing2LimitDegrees=20, twistLimitDegrees=10}
 {bone1Name=dyn_strand_1, bone2Name=head, swing1LimitDegrees=20, swing2LimitDegrees=20, twistLimitDegrees=5}
 {bone1Name=dyn_strand_2, bone2Name=dyn_strand_1, swing1LimitDegrees=20, swing2LimitDegrees=20, twistLimitDegrees=5}
 {bone1Name=dyn_strand_3, bone2Name=dyn_strand_2, swing1LimitDegrees=20, swing2LimitDegrees=20, twistLimitDegrees=5}
 {bone1Name=dyn_strand_4, bone2Name=dyn_strand_3, swing1LimitDegrees=20, swing2LimitDegrees=20, twistLimitDegrees=5}
 {bone1Name=dyn_backHair, bone2Name=head, swing1LimitDegrees=6, swing2LimitDegrees=10, twistLimitDegrees=45, twistMotion=Locked}
 {
  bone1Name=head
  bone2Name=spine_05
  swing1LimitDegrees=45
  swing1Motion=Free
  swing2LimitDegrees=45
  swing2Motion=Free
  twistLimitDegrees=45
  twistMotion=Free
 }
 {
  bone1Name=clavicle_l
  bone2Name=spine_05
  swing1LimitDegrees=45
  swing1Motion=Free
  swing2LimitDegrees=45
  swing2Motion=Free
  twistLimitDegrees=45
  twistMotion=Free
 }
 {
  bone1Name=clavicle_r
  bone2Name=spine_05
  swing1LimitDegrees=45
  swing1Motion=Free
  swing2LimitDegrees=45
  swing2Motion=Free
  twistLimitDegrees=45
  twistMotion=Free
 }
 {
  bone1Name=dyn_R_side_hair
  bone2Name=head
  swing1LimitDegrees=10
  swing1Motion=Locked
  swing2LimitDegrees=45
  twistLimitDegrees=45
  twistMotion=Locked
 }
]
