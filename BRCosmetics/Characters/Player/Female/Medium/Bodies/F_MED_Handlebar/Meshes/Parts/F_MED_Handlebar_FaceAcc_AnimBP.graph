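# Animation Blueprint for the face accessory mesh of the female medium Handlebar player cosmetic, adding secondary motion to a copied pose.
# It layers the owning component's pose from the upper arms down onto the attached parent's pose and curves.
# RigidBody then simulates in spine_05 space with custom gravity.
# Neck yaw relative to spine_05 drives an additive rotation and scale on dyn_backStrand after physics, likely as a corrective.
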AnimBlueprint /BRCosmetics/Characters/Player/Female/Medium/Bodies/F_MED_Handlebar/Meshes/Parts/F_MED_Handlebar_FaceAcc_AnimBP
target_skeleton=&/F_MED_Handlebar_FaceAcc_Skeleton
graphs [
 AnimGraph {nodes [
   n1 = Copy Pose From Mesh(SourceMeshComponent=Get Body_Mesh())
   n2 = Copy Pose From Mesh() {bCopyCurves=true, bUseAttachedParent=true}
   n3 = RigidBody!(
    ComponentPose=Local To Component(LocalPose=Layered blend per bone(BasePose=@n2, BlendPoses_0=@n1) {layerSetup [{branchFilters [{boneName=upperarm_l}, {boneName=upperarm_r}]}]})
    Alpha
    OverrideWorldGravity:Vector="0,0,-980"
    ExternalForce
   ) -> (Pose) {
    bClampLinearTranslationLimitToRefPose=true
    bOverrideWorldGravity=true
    overridePhysicsAsset=&/F_MED_Handlebar_FaceAcc_Physics
    simulationSpace=BaseBoneSpace
    alphaBoolBlend {blendOption=Linear}
    baseBoneRef {boneName=spine_05}
    componentAppliedLinearAccClamp=<1000 1000 1000>
    componentLinearAccScale=<0.2 0.2 0.3>
    componentLinearVelScale=<0.1 0.1 0.2>
   }
   n4 = "Transform (Modify) Bone - Bone: dyn_backStrand"(
    ComponentPose=@n3
    Alpha=Get neckBackAlpha() -> (neckBackAlpha:double)
    Translation="0,0,0"
    Rotation="-6,-0.000,-0.000"
    Scale="0.8,0.7,0.7"
   ) {
    rotationMode=BMM_Additive
    rotationSpace=BCS_BoneSpace
    scaleMode=BMM_Additive
    scaleSpace=BCS_BoneSpace
    translationSpace=BCS_BoneSpace
    alphaBoolBlend {blendOption=Linear}
    boneToModify {boneName=dyn_backStrand}
    rotation {pitch=-6}
    scale=<0.8 0.7 0.7>
   }
   Output Pose(Result=Component To Local(ComponentPose=@n4)) {layerGroup=None}
  ]}
 EventGraph {nodes [
   Event BlueprintUpdateAnimation()
   n1 = BreakTransform(InTransform=GetDeltaTransformFromRefPose(self=GetOwnerSkeletalMeshComponent(), BoneName=neck_01, BaseName=spine_05))
   > Set neckBackAlpha(neckBackAlpha:double=MapRangeClamped(Value=BreakRotator(InRot=@n1.Rotation).Yaw, InRangeA=0, InRangeB=-30)) -> (Output_Get:double)
   Event BlueprintInitializeAnimation()
   > Set Body_Mesh(Body_Mesh=GetOwnerSkeletalMeshComponent())
  ]}
]
variables [
 Body_Mesh {type=SkeletalMeshComponent}
 neckBackAlpha {type=number}
]
